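# UEFN stylable button widget, a child of WBP_FNE_StylableButton that exposes five fixed image-layer slots, each with material, appearance, per-state and transition settings.
# On PreConstruct it clears a scratch array, copies in each layer that has a valid material, and assigns it to the parent's ImageLayers before running parent PreConstruct.
# Only the first layer has a default material, a light UEFN button instance.
# It adds no widgets of its own.
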WidgetBlueprint /FNE_UI/UILibrary/Buttons/UEFN_Button_Stylable
graphs [EventGraph {nodes [
   Event Construct()
   > Parent: Construct()
   n1 = Event PreConstruct()
   > Clear(TargetArray:Array of S FNEUI Stylable Button Image Layers Settings Structures=Get ImageArray() -> (ImageArray:Array of S FNEUI Stylable Button Image Layers Settings Structures))
   n2 = Make Array!(
    "[0]":S FNEUI Stylable Button Image Layers Settings Structure=Get ImageLayer1() -> (ImageLayer1:S FNEUI Stylable Button Image Layers Settings Structure)
    "[1]":S FNEUI Stylable Button Image Layers Settings Structure=Get ImageLayer2() -> (ImageLayer2:S FNEUI Stylable Button Image Layers Settings Structure)
    "[2]":S FNEUI Stylable Button Image Layers Settings Structure=Get ImageLayer3() -> (ImageLayer3:S FNEUI Stylable Button Image Layers Settings Structure)
    "[3]":S FNEUI Stylable Button Image Layers Settings Structure=Get ImageLayer4() -> (ImageLayer4:S FNEUI Stylable Button Image Layers Settings Structure)
    "[4]":S FNEUI Stylable Button Image Layers Settings Structure=Get ImageLayer5() -> (ImageLayer5:S FNEUI Stylable Button Image Layers Settings Structure)
   ) -> (Array:Array of S FNEUI Stylable Button Image Layers Settings Structures)
   > n3 = For Each Loop!(Exec, Array:Array of S FNEUI Stylable Button Image Layers Settings Structures=@n2) -> (
    LoopBody
    Array Element_Material_51_#1:Material Instance Object Reference
    Array Element_Settings_50_#2:S FNEUI Stylable Button Image Layers Settings Advanced Structure
    Array Element_States_54_#3_BaseStates_6_#4:S FNEUI States Base Structure
    Array Element_States_54_#3_TransitionDurationOverrides_7_#5:S FNEUI States Transition Durations Override Structure
    Array Index
    Completed
   )
   Set ImageLayers(execute=@n3.Completed, ImageLayers:Array of S FNEUI Stylable Button Image Layers Settings Structures=Get ImageArray()) -> (Output_Get:Array of S FNEUI Stylable Button Image Layers Settings Structures)
   > Parent: PreConstruct(IsDesignTime=@n1.IsDesignTime)
   Branch(execute=@n3.LoopBody, Condition=IsValid(Object=@n3))
   > Add(
    TargetArray:Array of S FNEUI Stylable Button Image Layers Settings Structures=Get ImageArray()
    NewItem:"S FNEUI Stylable Button Image Layers Settings Structure (by ref)"="Get (a ref)"(Array:Array of S FNEUI Stylable Button Image Layers Settings Structures=@n2, Dimension 1=@n3.Array Index) -> (Output:"S FNEUI Stylable Button Image Layers Settings Structure (by ref)")
   ) {is=title K2Node_CallArrayFunction}
   Event Destruct()
   > Parent: Destruct()
  ]}]
variables [
 ImageArray {type="array<S_FNEUI_StylableButton_ImageLayers_Settings>", default []}
 ImageLayer1 {
  type=S_FNEUI_StylableButton_ImageLayers_Settings
  default {
   material=/FortUILibrary/UGC/MaterialInstances/MI_UEFN_UI_Button_Light
   settings {
    angle=0
    draw as=Image
    horizontal Alignment=HAlign_Fill
    vertical Alignment=VAlign_Fill
    visibility=HitTestInvisible
    color and Opacity=<rgba 1 1 1 1>
    padding=<ltrb 0 0 0 0>
    pivot=<0.5 0.5>
    rounded Box and Outline Settings {"outline Thickness (or Gap)"=0, rounding Type=Half Height Radius, switch to Outline=false, corner Radius=<0 0 0 0>}
    scale=<1 1>
    shear=<0 0>
    size=<32 32>
    stretch Margin=<ltrb 0.5 0.5 0.5 0.5>
    translation=<0 0>
   }
   states {
    base States {
     disabled {angle=0, enable State=false, color and Opacity=<rgba 1 1 1 1>, scale=<1 1>, shear=<0 0>, translation=<0 0>}
     focused {angle=0, enable State=false, color and Opacity=<rgba 1 1 1 1>, scale=<1 1>, shear=<0 0>, translation=<0 0>}
     hovered {angle=0, enable State=false, color and Opacity=<rgba 1 1 1 1>, scale=<1 1>, shear=<0 0>, translation=<0 0>}
     pressed {angle=0, enable State=false, color and Opacity=<rgba 1 1 1 1>, scale=<1 1>, shear=<0 0>, translation=<0 0>}
    }
    transition Duration Overrides {apply Overrides=false, disabling=0, enabling=0, focusing=0, hovering=0, pressing=0, releasing=0, unfocusing=0, unhovering=0}
   }
  }
 }
 ImageLayer2 {
  type=S_FNEUI_StylableButton_ImageLayers_Settings
  default {
   material=null
   settings {
    angle=0
    draw as=Image
    horizontal Alignment=HAlign_Fill
    vertical Alignment=VAlign_Fill
    visibility=HitTestInvisible
    color and Opacity=<rgba 1 1 1 1>
    padding=<ltrb 0 0 0 0>
    pivot=<0.5 0.5>
    rounded Box and Outline Settings {"outline Thickness (or Gap)"=0, rounding Type=Half Height Radius, switch to Outline=false, corner Radius=<0 0 0 0>}
    scale=<1 1>
    shear=<0 0>
    size=<32 32>
    stretch Margin=<ltrb 0.5 0.5 0.5 0.5>
    translation=<0 0>
   }
   states {
    base States {
     disabled {angle=0, enable State=false, color and Opacity=<rgba 1 1 1 1>, scale=<1 1>, shear=<0 0>, translation=<0 0>}
     focused {angle=0, enable State=false, color and Opacity=<rgba 1 1 1 1>, scale=<1 1>, shear=<0 0>, translation=<0 0>}
     hovered {angle=0, enable State=false, color and Opacity=<rgba 1 1 1 1>, scale=<1 1>, shear=<0 0>, translation=<0 0>}
     pressed {angle=0, enable State=false, color and Opacity=<rgba 1 1 1 1>, scale=<1 1>, shear=<0 0>, translation=<0 0>}
    }
    transition Duration Overrides {apply Overrides=false, disabling=0, enabling=0, focusing=0, hovering=0, pressing=0, releasing=0, unfocusing=0, unhovering=0}
   }
  }
 }
 ImageLayer3 {
  type=S_FNEUI_StylableButton_ImageLayers_Settings
  default {
   material=null
   settings {
    angle=0
    draw as=Image
    horizontal Alignment=HAlign_Fill
    vertical Alignment=VAlign_Fill
    visibility=HitTestInvisible
    color and Opacity=<rgba 1 1 1 1>
    padding=<ltrb 0 0 0 0>
    pivot=<0.5 0.5>
    rounded Box and Outline Settings {"outline Thickness (or Gap)"=0, rounding Type=Half Height Radius, switch to Outline=false, corner Radius=<0 0 0 0>}
    scale=<1 1>
    shear=<0 0>
    size=<32 32>
    stretch Margin=<ltrb 0.5 0.5 0.5 0.5>
    translation=<0 0>
   }
   states {
    base States {
     disabled {angle=0, enable State=false, color and Opacity=<rgba 1 1 1 1>, scale=<1 1>, shear=<0 0>, translation=<0 0>}
     focused {angle=0, enable State=false, color and Opacity=<rgba 1 1 1 1>, scale=<1 1>, shear=<0 0>, translation=<0 0>}
     hovered {angle=0, enable State=false, color and Opacity=<rgba 1 1 1 1>, scale=<1 1>, shear=<0 0>, translation=<0 0>}
     pressed {angle=0, enable State=false, color and Opacity=<rgba 1 1 1 1>, scale=<1 1>, shear=<0 0>, translation=<0 0>}
    }
    transition Duration Overrides {apply Overrides=false, disabling=0, enabling=0, focusing=0, hovering=0, pressing=0, releasing=0, unfocusing=0, unhovering=0}
   }
  }
 }
 ImageLayer4 {
  type=S_FNEUI_StylableButton_ImageLayers_Settings
  default {
   material=null
   settings {
    angle=0
    draw as=Image
    horizontal Alignment=HAlign_Fill
    vertical Alignment=VAlign_Fill
    visibility=HitTestInvisible
    color and Opacity=<rgba 1 1 1 1>
    padding=<ltrb 0 0 0 0>
    pivot=<0.5 0.5>
    rounded Box and Outline Settings {"outline Thickness (or Gap)"=0, rounding Type=Half Height Radius, switch to Outline=false, corner Radius=<0 0 0 0>}
    scale=<1 1>
    shear=<0 0>
    size=<32 32>
    stretch Margin=<ltrb 0.5 0.5 0.5 0.5>
    translation=<0 0>
   }
   states {
    base States {
     disabled {angle=0, enable State=false, color and Opacity=<rgba 1 1 1 1>, scale=<1 1>, shear=<0 0>, translation=<0 0>}
     focused {angle=0, enable State=false, color and Opacity=<rgba 1 1 1 1>, scale=<1 1>, shear=<0 0>, translation=<0 0>}
     hovered {angle=0, enable State=false, color and Opacity=<rgba 1 1 1 1>, scale=<1 1>, shear=<0 0>, translation=<0 0>}
     pressed {angle=0, enable State=false, color and Opacity=<rgba 1 1 1 1>, scale=<1 1>, shear=<0 0>, translation=<0 0>}
    }
    transition Duration Overrides {apply Overrides=false, disabling=0, enabling=0, focusing=0, hovering=0, pressing=0, releasing=0, unfocusing=0, unhovering=0}
   }
  }
 }
 ImageLayer5 {
  type=S_FNEUI_StylableButton_ImageLayers_Settings
  default {
   material=null
   settings {
    angle=0
    draw as=Image
    horizontal Alignment=HAlign_Fill
    vertical Alignment=VAlign_Fill
    visibility=HitTestInvisible
    color and Opacity=<rgba 1 1 1 1>
    padding=<ltrb 0 0 0 0>
    pivot=<0.5 0.5>
    rounded Box and Outline Settings {"outline Thickness (or Gap)"=0, rounding Type=Half Height Radius, switch to Outline=false, corner Radius=<0 0 0 0>}
    scale=<1 1>
    shear=<0 0>
    size=<32 32>
    stretch Margin=<ltrb 0.5 0.5 0.5 0.5>
    translation=<0 0>
   }
   states {
    base States {
     disabled {angle=0, enable State=false, color and Opacity=<rgba 1 1 1 1>, scale=<1 1>, shear=<0 0>, translation=<0 0>}
     focused {angle=0, enable State=false, color and Opacity=<rgba 1 1 1 1>, scale=<1 1>, shear=<0 0>, translation=<0 0>}
     hovered {angle=0, enable State=false, color and Opacity=<rgba 1 1 1 1>, scale=<1 1>, shear=<0 0>, translation=<0 0>}
     pressed {angle=0, enable State=false, color and Opacity=<rgba 1 1 1 1>, scale=<1 1>, shear=<0 0>, translation=<0 0>}
    }
    transition Duration Overrides {apply Overrides=false, disabling=0, enabling=0, focusing=0, hovering=0, pressing=0, releasing=0, unfocusing=0, unhovering=0}
   }
  }
 }
]
widgets {
 info {
  inheritedWidgetCount=0
  parentClass=&/WBP_FNE_StylableButton.WBP_FNE_StylableButton_C
  rootWidgetClass=null
  widgetCount=0
 }
 widgets []
}
